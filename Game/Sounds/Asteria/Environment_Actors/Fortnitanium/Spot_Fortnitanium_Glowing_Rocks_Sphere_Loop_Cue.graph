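# Looping outdoor ambient cue for glowing Fortnitanium rocks, with its own short-range attenuation, listener focus and a tight concurrency limit.
# Quality levels select either a random pick of three modulated branches, two with random start delays, or a single undelayed branch.
# Each randomized branch fades in a looping rock wave.
# Two nested, endlessly looping envelopes of different lengths slowly vary pitch and swell volume over the result.
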SoundCue /Game/Sounds/Asteria/Environment_Actors/Fortnitanium/Spot_Fortnitanium_Glowing_Rocks_Sphere_Loop_Cue
root=Enveloper_3
nodes [
 {id=Delay_0, children [WavePlayer_1], properties {delayMax=2}}
 {id=Delay_1, children [WavePlayer_2], properties {delayMax=3, delayMin=2}}
 {
  id=Enveloper_0
  children [QualityLevel_1]
  properties {
   bLoop=true
   bLoopIndefinitely=true
   loopEnd=30
   pitchCurve {editorCurveData {keys [{arriveTangent=-0.0013, interpMode=RCIM_Cubic, time=0}, <at 15 0.95 RCIM_Cubic>, {interpMode=RCIM_Cubic, leaveTangent=0.0013, time=30}]}}
   volumeCurve {editorCurveData {keys [<at 0 1>]}}
  }
 }
 {
  id=Enveloper_1
  children [Modulator_1]
  properties {pitchCurve {editorCurveData {keys [{time=0}]}}, volumeCurve {editorCurveData {keys [<at 0 0>,<at 1 1>]}}}
 }
 {
  id=Enveloper_2
  children [Modulator_2]
  properties {pitchCurve {editorCurveData {keys [{time=0}]}}, volumeCurve {editorCurveData {keys [<at 0 0>,<at 1 1>]}}}
 }
 {
  id=Enveloper_3
  children [Enveloper_0]
  properties {
   bLoop=true
   bLoopIndefinitely=true
   loopEnd=20
   pitchCurve {editorCurveData {keys [{time=0}]}}
   volumeCurve {editorCurveData {keys [
      {arriveTangent=-0.01, interpMode=RCIM_Cubic, time=0, value=1}
      <at 10 0.5 RCIM_Cubic>
      {interpMode=RCIM_Cubic, leaveTangent=0.01, time=20, value=1}
     ]}}
  }
 }
 {
  id=Enveloper_4
  children [Modulator_0]
  properties {pitchCurve {editorCurveData {keys [{time=0}]}}, volumeCurve {editorCurveData {keys [<at 0 0>,<at 1 1>]}}}
 }
 {id=Modulator_0, children [WavePlayer_0]}
 {id=Modulator_1, children [Delay_0]}
 {id=Modulator_2, children [Delay_1]}
 {id=QualityLevel_1, children [Random_0,Random_0,Modulator_0,Enveloper_4]}
 {id=Random_0, children [Enveloper_1,Enveloper_2,Enveloper_4], properties {weights [1,1,1]}}
 {
  id=WavePlayer_0
  properties {bLooping=true, soundWaveAssetPtr=&/Spot_Fornitanium_Glowing_Rock_Loop}
 }
 {
  id=WavePlayer_1
  properties {bLooping=true, soundWaveAssetPtr=&Spot_Fornitanium_Glowing_Rock_Loop}
 }
 {
  id=WavePlayer_2
  properties {bLooping=true, soundWaveAssetPtr=&Spot_Fornitanium_Glowing_Rock_Loop}
 }
]
properties {
 bOverrideAttenuation=true
 bOverrideConcurrency=true
 maxDistance=750
 soundClassObject=&Fort_Ambient_Outdoors_Athena
 volumeMultiplier=0.652
 attenuationOverrides {
  bEnableFocusInterpolation=true
  bEnableListenerFocus=true
  falloffDistance=500
  focusAttackInterpSpeed=0.01
  focusReleaseInterpSpeed=0.01
  focusVolumeAttenuation=1.2
  nonFocusVolumeAttenuation=0.75
  attenuationShapeExtents=<250 500>
 }
 concurrencyOverrides {maxCount=2, volumeScaleMode=Distance}
}
